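# Animation Blueprint for the M_MED_BackAbsorb male character body, adding secondary motion to its suit jacket and tie.
# It copies the parent mesh pose into a Control Rig with stance flags, then caches it.
# RigidBody nodes simulate the jacket, switching physics assets for skydiving, then the tie, followed by additive bone corrections.
# The event graph updates a parameter-driven gravity override and a crouching-while-moving flag.
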AnimBlueprint /BRCosmetics/Characters/Player/Male/Medium/Bodies/M_MED_BackAbsorb/Meshes/M_MED_BackAbsorb_AnimBP
target_skeleton=&/M_MED_BackAbsorb_Skeleton
graphs [
 AnimGraph {nodes [
   n1 = RigidBody!(ComponentPose=Local To Component(LocalPose=Use cached pose 'PreSimPose'()), Alpha, OverrideWorldGravity:Vector="0,0,0", ExternalForce) -> (Pose) {
    bOverrideWorldGravity=true
    overridePhysicsAsset=&/M_MED_BackAbsorb_Jacket_Skydive_Physics
    simulationSpace=BaseBoneSpace
    baseBoneRef {boneName=pelvis}
    componentAppliedLinearAccClamp=<1000 1000 1000>
    simSpaceSettings {
     dampingAlpha=0.6
     maxAngularAcceleration=20
     maxAngularVelocity=5
     maxLinearAcceleration=1000
     maxLinearVelocity=300
     velocityScaleZ=0.5
     worldAlpha=0.5
     externalLinearDragV=<0.3 0.3 0.3>
    }
   }
   n2 = Component To Local(ComponentPose=@n1)
   n3 = RigidBody(ComponentPose=Local To Component(LocalPose=Use cached pose 'PreSimPose'()), OverrideWorldGravity="0,0,0") {
    bOverrideWorldGravity=true
    overridePhysicsAsset=&/M_MED_BackAbsorb_Jacket_Physics
    simulationSpace=BaseBoneSpace
    baseBoneRef {boneName=pelvis}
    componentAppliedLinearAccClamp=<3000 3000 3000>
    simSpaceSettings {
     dampingAlpha=0.1
     maxAngularAcceleration=30
     maxAngularVelocity=6
     maxLinearAcceleration=1000
     maxLinearVelocity=300
     velocityScaleZ=0.5
     worldAlpha=0.6
     externalLinearDragV=<0.3 0.3 1>
    }
   }
   n4 = Component To Local(ComponentPose=@n3)
   n5 = Get bIsSkyDiving()
   n6 = RigidBody(
    ComponentPose=Local To Component(LocalPose=Blend Poses by bool(BlendPose_0=@n2, BlendPose_1=@n4, bActiveValue=@n5) {blendType=Linear})
    OverrideWorldGravity="0,0,0"
   ) {
    bOverrideWorldGravity=true
    overridePhysicsAsset=&/M_MED_BackAbsorb_Tie_Physics
    simulationSpace=BaseBoneSpace
    baseBoneRef {boneName=spine_05}
    componentAppliedLinearAccClamp=<1000 1000 1000>
    simSpaceSettings {
     dampingAlpha=0.5
     maxAngularAcceleration=20
     maxAngularVelocity=5
     maxLinearAcceleration=1000
     maxLinearVelocity=300
     velocityScaleZ=0.25
     worldAlpha=1
     externalLinearDragV=<0.3 0.3 0.3>
    }
   }
   n7 = "Transform (Modify) Bone - Bone: base_suitSideFront_l"!(ComponentPose=@n6, bAlphaBoolEnabled:bool=True, Translation="0.000,0.000,-1.302", Rotation="1.589,0.000,0.000", Scale="1,1,1") -> (Pose) {
    alphaInputType=Bool
    rotationMode=BMM_Additive
    translationMode=BMM_Additive
    alphaBoolBlend {blendInTime=0.5, blendOutTime=0.5}
    boneToModify {boneName=base_suitSideFront_l}
    rotation {pitch=1.589}
    translation {z=-1.302}
   }
   n8 = "Transform (Modify) Bone - Bone: dyn_suitSideFront_01_r"!(ComponentPose=@n7, bAlphaBoolEnabled:bool=True, Translation="0,0,0", Rotation="-0.793,0.000,-0.000", Scale="1,1,1") -> (Pose) {
    alphaInputType=Bool
    rotationMode=BMM_Additive
    translationMode=BMM_Additive
    alphaBoolBlend {blendInTime=0.5, blendOutTime=0.5}
    boneToModify {boneName=dyn_suitSideFront_01_r}
    rotation {pitch=-0.793}
   }
   n9 = "Transform (Modify) Bone - Bone: dyn_suitSide_02_l"!(ComponentPose=@n8, bAlphaBoolEnabled:bool=True, Translation="0,0,0", Rotation="-15.940,-9.328,-0.000", Scale="1,1,1") -> (Pose) {
    alphaInputType=Bool
    rotationMode=BMM_Additive
    translationMode=BMM_Additive
    alphaBoolBlend {blendInTime=0.5, blendOutTime=0.5}
    boneToModify {boneName=dyn_suitSide_02_l}
    rotation=<rot -15.94 -9.328 0>
   }
   n10 = "Transform (Modify) Bone - Bone: dyn_suitSideFront_02_l"!(ComponentPose=@n9, bAlphaBoolEnabled:bool=True, Translation="2.238,0.000,0.000", Rotation="-35.003,-30.329,14.571", Scale="1,1,1") -> (Pose) {
    alphaInputType=Bool
    rotationMode=BMM_Additive
    translationMode=BMM_Additive
    alphaBoolBlend {blendInTime=0.5, blendOutTime=0.5}
    boneToModify {boneName=dyn_suitSideFront_02_l}
    rotation=<rot -35.003 -30.329 14.571>
    translation {x=2.238}
   }
   n11 = "Transform (Modify) Bone - Bone: dyn_suitSideFront_02_l"(ComponentPose=@n10, bAlphaBoolEnabled=True, Translation="0,0,0", Rotation="24.648,26,0", Scale="1,1,1") {
    alphaInputType=Bool
    rotationMode=BMM_Additive
    translationMode=BMM_Additive
    alphaBoolBlend {blendInTime=0.5, blendOutTime=0.5}
    boneToModify {boneName=dyn_suitSideFront_02_l}
    rotation {pitch=24.648, yaw=26}
   }
   n12 = "Transform (Modify) Bone - Bone: dyn_suitSide_02_l"(ComponentPose=@n11, bAlphaBoolEnabled=True, Translation="0,0,0", Rotation="19.635,-9.890,4.531", Scale="1,1,1") {
    alphaInputType=Bool
    rotationMode=BMM_Additive
    translationMode=BMM_Additive
    alphaBoolBlend {blendInTime=0.5, blendOutTime=0.5}
    boneToModify {boneName=dyn_suitSide_02_l}
    rotation=<rot 19.635 -9.89 4.531>
   }
   n13 = "Transform (Modify) Bone - Bone: dyn_tie_02"!(ComponentPose=@n12, bAlphaBoolEnabled:bool=True, Translation="0,0,0", Rotation="0.229,-0.027,-13.027", Scale="1,1,1") -> (Pose) {
    alphaInputType=Bool
    rotationMode=BMM_Additive
    translationMode=BMM_Additive
    alphaBoolBlend {blendInTime=1, blendOutTime=1}
    boneToModify {boneName=dyn_tie_02}
    rotation=<rot 0.229 -0.027 -13.027>
   }
   n14 = "Transform (Modify) Bone - Bone: dyn_tie_03"!(ComponentPose=@n13, bAlphaBoolEnabled:bool=True, Translation="0,0,0", Rotation="0.229,-0.019,-12.235", Scale="1,1,1") -> (Pose) {
    alphaInputType=Bool
    rotationMode=BMM_Additive
    translationMode=BMM_Additive
    alphaBoolBlend {blendInTime=1, blendOutTime=1}
    boneToModify {boneName=dyn_tie_03}
    rotation=<rot 0.229 -0.019 -12.235>
   }
   n15 = "Transform (Modify) Bone - Bone: dyn_suitSideFront_01_r"(ComponentPose=@n14, bAlphaBoolEnabled=True, Translation="0,0,0", Rotation="0.017,-4.922,0.032", Scale="1,1,1") {
    alphaInputType=Bool
    rotationMode=BMM_Additive
    rotationSpace=BCS_BoneSpace
    scaleSpace=BCS_BoneSpace
    translationMode=BMM_Additive
    translationSpace=BCS_BoneSpace
    alphaBoolBlend {blendInTime=0.2, blendOutTime=0.2}
    boneToModify {boneName=dyn_suitSideFront_01_r}
    rotation=<rot 0.017 -4.922 0.032>
   }
   n16 = "Transform (Modify) Bone - Bone: base_suitSide_r"!(ComponentPose=@n15, bAlphaBoolEnabled:bool=True, Translation="0,0,0", Rotation="-9.299,-4.376,-0.007", Scale="1,1,1") -> (Pose) {
    alphaInputType=Bool
    rotationMode=BMM_Additive
    rotationSpace=BCS_BoneSpace
    scaleSpace=BCS_BoneSpace
    translationMode=BMM_Additive
    translationSpace=BCS_BoneSpace
    alphaBoolBlend {blendInTime=0.6, blendOutTime=1}
    boneToModify {boneName=base_suitSide_r}
    rotation=<rot -9.299 -4.376 -0.007>
   }
   n17 = "Transform (Modify) Bone - Bone: dyn_suitSideFront_02_r"!(ComponentPose=@n16, bAlphaBoolEnabled:bool=True, Translation="0,0,0", Rotation="4.978,16,0.7", Scale="1,1,1") -> (Pose) {
    alphaInputType=Bool
    rotationMode=BMM_Additive
    rotationSpace=BCS_BoneSpace
    scaleSpace=BCS_BoneSpace
    translationMode=BMM_Additive
    translationSpace=BCS_BoneSpace
    alphaBoolBlend {blendInTime=0.6, blendOutTime=1}
    boneToModify {boneName=dyn_suitSideFront_02_r}
    rotation=<rot 4.978 16 0.7>
   }
   n18 = "Transform (Modify) Bone - Bone: base_suitSideFront_r"!(ComponentPose=@n17, bAlphaBoolEnabled:bool=True, Translation="0,0,0", Rotation="0.974,-15,-0.059", Scale="1,1,1") -> (Pose) {
    alphaInputType=Bool
    rotationMode=BMM_Additive
    rotationSpace=BCS_BoneSpace
    scaleSpace=BCS_BoneSpace
    translationMode=BMM_Additive
    translationSpace=BCS_BoneSpace
    alphaBoolBlend {blendInTime=0.2, blendOutTime=1}
    boneToModify {boneName=base_suitSideFront_r}
    rotation=<rot 0.974 -15 -0.059>
   }
   n19 = "Transform (Modify) Bone - Bone: base_tie"!(ComponentPose=@n18, bAlphaBoolEnabled:bool=True, Translation="0,0,0", Rotation="0.237,0.000,-4.654", Scale="1,1,1") -> (Pose) {
    alphaInputType=Bool
    rotationMode=BMM_Additive
    translationMode=BMM_Additive
    alphaBoolBlend {blendInTime=1, blendOutTime=1}
    boneToModify {boneName=base_tie}
    rotation=<rot 0.237 0 -4.654>
   }
   n20 = "Transform (Modify) Bone - Bone: dyn_suitSideFront_01_l"!(ComponentPose=@n19, bAlphaBoolEnabled:bool=True, Translation="0,0,0", Rotation="0.010,-4.817,-2.829", Scale="1,1,1") -> (Pose) {
    alphaInputType=Bool
    rotationMode=BMM_Additive
    rotationSpace=BCS_BoneSpace
    scaleSpace=BCS_BoneSpace
    translationMode=BMM_Additive
    translationSpace=BCS_BoneSpace
    alphaBoolBlend {blendInTime=1, blendOutTime=1}
    boneToModify {boneName=dyn_suitSideFront_01_l}
    rotation=<rot 0.01 -4.817 -2.829>
   }
   n21 = "Transform (Modify) Bone - Bone: dyn_suitSideFront_01_r"(ComponentPose=@n20, bAlphaBoolEnabled=True, Translation="0,0,0", Rotation="0.010,-4.817,-2.829", Scale="1,1,1") {
    alphaInputType=Bool
    rotationMode=BMM_Additive
    rotationSpace=BCS_BoneSpace
    scaleSpace=BCS_BoneSpace
    translationMode=BMM_Additive
    translationSpace=BCS_BoneSpace
    alphaBoolBlend {blendInTime=1, blendOutTime=1}
    boneToModify {boneName=dyn_suitSideFront_01_r}
    rotation=<rot 0.01 -4.817 -2.829>
   }
   Output Pose(Result=Component To Local(ComponentPose=@n21)) {layerGroup=None}
   n22 = Control Rig!(
    Source=Copy Pose From Mesh() {bUseAttachedParent=true}
    Alpha
    isCrouching_CR:bool=Get bIsCrouching()
    isKneeSliding_CR:bool=Get bIsKneeSliding()
    isBaseballSliding_CR:bool=Get bIsBaseballSliding()
    isFE_CR:bool=Get bIsInFrontEnd()
    isDBNO_CR:bool=Get bIsDBNO()
   ) -> (Pose) {controlRigAssetReference {blueprintRigClass=&/M_MED_BackAbsorb_CtrlRig.M_MED_BackAbsorb_CtrlRig_C}}
   Save cached pose 'PreSimPose'(Pose=@n22) {cacheName=PreSimPose}
  ]}
 EventGraph {nodes [
   Event BlueprintUpdateAnimation()
   > n1 = "Gravity Override State Function (Parameter Asset) (Nativized)"(
    CharacterPartAnimInstance=Self-Reference()
    GameTimeInSeconds=GetGameTimeInSeconds()
    DebugJointName=spine_05
    Parameters=&/M_MED_BackAbsorb_Parameters
   )
   > Set Gravity Override(Gravity Override=@n1)
   > Set isCrouchingAndMoving(isCrouchingAndMoving:bool=AND Boolean(A=Get bIsCrouching(), B=Get bIsMoving2D())) -> (Output_Get:bool)
  ]}
]
variables [
 Gravity Override {type=Vector}
 isCrouchingAndMoving {type=boolean}
]
